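; Junk variant of the flopper fishing spawn actor, with its loot tier forced to the junk override.
; Spawning usually defers to the parent, but a roughly one-in-twenty roll spawns a tiered flopper chest instead.
; On that path, after a delay, the server attaches the chest to the hook if hooked and tracks uses.
; For rod or Happy Ghost catches it also ends the fishing session with telemetry, then plays a world-spawner stinger or destroys itself.
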(event EventSpawnItem
  (params (Target :type Actor%20Object%20Reference) (Hooked :type Boolean) (HookComponent :type Scene%20Component%20Object%20Reference) (SpawnLoc :type Vector :default "0, 0, 0") (Player :type Actor%20Object%20Reference) (ItemSpawnDelay :type Float%20%28double-precision%29) (CaughtWithFishingRod :type Boolean) (CaughtWithHappyGhost :type Boolean) (ItemUsedToFish :type Actor%20Object%20Reference) (Projectile :type Object%20Reference) (Required%20Tags :type Gameplay%20Tag%20Container%20Structure))
  ;[1 1 1] Spawn Items
  (if (>= (Math|Random|RandomFloatinRange 0.0 1.0) 0.05)
    (|CallParentFunction :Target Target :Hooked Hooked :HookComponent HookComponent :SpawnLoc SpawnLoc :Player Player :ItemSpawnDelay ItemSpawnDelay :CaughtWithFishingRod CaughtWithFishingRod :CaughtWithHappyGhost CaughtWithHappyGhost :ItemUsedToFish ItemUsedToFish :Projectile Projectile)
    (else
      (Variables|Default|SetProjectile Projectile)
      (Development|PrintString "FlopperSpawnItems" (Variables|Default|GetDebug) (Variables|Default|GetDebug))
      (Utilities|FlowControl|Delay ItemSpawnDelay) ; Used to delay the item spawn so that the water does not push the item around when it spawns
      (if (Networking|IsServer)
        (if (not (Variables|Default|GetMaxUses))
          (Default|CallOnFishingComplete)
          (Utilities|FlowControl|Sequence
            (:then_0
              (bind _returnvalue (Game|SpawnActorfromClass "/Game/Athena/Items/Consumables/Flopper/Tiered_Chest_Athena_Flopper_Proto.Tiered_Chest_Athena_Flopper_Proto_C" (Math|Transform|MakeTransform (select Hooked SpawnLoc (Transformation|GetActorLocation))) "AdjustIfPossibleButAlwaysSpawn" "MultiplyWithRoot" 0 0 0 (Game|GetInstigator)))
              (if Hooked
                (Transformation|AttachActorToComponent _returnvalue HookComponent "None" "SnapToTarget" "KeepWorld" "KeepWorld" true)))
            (:then_1
              (CallFunction|HandleUseTracking)
              ;[1 1 1] End Fishing Event if caught from fishing rod
              (if (or CaughtWithFishingRod CaughtWithHappyGhost)
                (bind _asfort_player_pawn (Utilities|Casting|CastToFortPlayerPawn Player)
                  (:then
                    (Fortnite|Fishing|FireEventEndFishingSessionwithData _asfort_player_pawn (Utilities|Struct|MakePlayerFishingTelemetryData (Variables|Default|GetSpawnedItem) ItemUsedToFish (Variables|Default|GetIsWorldSpawner) false "" (Variables|CollectionAnalytics|GetFishPoolTag))))
                  (:CastFailed
                    _))
                ;[1 1 1] Play Stinger if item was spawned from World Spawner
                (if (Variables|Default|GetIsWorldSpawner)
                  (Networking|FlushNetDormancy)
                  (CallFunction|PlaySpawnItemHightier (Transformation|GetActorLocation Player))
                  (else
                    (Actor|DestroyActor)))
                (else
                  _)))))))))

(event Custom|CustomEvent_1
  (params (Location :type Vector))
  ;[1 1 1] Play Stinger if item was spawned from World Spawner
  (Audio|PlaySoundatLocation (Variables|Default|GetSound_HighTierPoolSpawn) Location))

(fn GetLootTier
  (params (IsExplosion :type Boolean) (IsProFishingRod :type Boolean) (Player :type Actor%20Object%20Reference))
  (bind _overrideloottier_junk (Variables|Default|GetOverrideLootTier_Junk))
  (Variables|Items|SetItemsToDrop _overrideloottier_junk)
  (return _overrideloottier_junk))

(fn ConstructionScript
  (|CallParentFunction))
