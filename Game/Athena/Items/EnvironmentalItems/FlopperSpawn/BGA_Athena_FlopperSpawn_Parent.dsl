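; Parent Blueprint for fishing spots that yield fish and other loot, either as world spawners or default spawns.
; When playlist data is ready it rolls uses, culls world spawners by chance, and server-side reads POI loot tags and nudges itself off shore.
; SpawnItem picks loot by tags, rod tier or player override component, spawns pickups, and feeds collections, quests and telemetry.
; Depleted spots hide, then respawn or destroy themselves.
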(event Custom|HideAndDelayForRespawn
  (params (TimeToDelay :type Float%20%28double-precision%29))
  ;[1 1 1] Hide and Delay then respawn with new uses
  (if (and (> TimeToDelay 0.0) (Variables|Default|GetIsWorldSpawner))
    (Rendering|SetActorHiddenInGame true)
    (Collision|SetActorEnableCollision false)
    (Utilities|FlowControl|Delay TimeToDelay)
    (CallFunction|SetBalanceValues)
    (Variables|Default|SetMaxUses false)
    (Rendering|SetActorHiddenInGame false)
    (Collision|SetActorEnableCollision true)
    (else
      (CallFunction|HideandKill) ; shouldn't normally happen saftey check fallback
    )))

(event Custom|HideAndKill
  ;[1 1 1] Hide and Kill
  (CallFunction|DestroyFlopperSpawn))

(event EventBeginPlay
  ;[1 1 1] Begin Play and set sheet values and check if this one should be removed
  (Class|FortAsyncActionCurrentPlaylistReady|PlaylistDataReadyAsync
    (:OnReady
      (CallFunction|SetBalanceValues)
      (if (Variables|Default|GetIsWorldSpawner)
        (CallFunction|ShouldKill)
        (Variables|Items|SetTagToLoot_ToUse (Variables|Items|GetTagToLoot_FishingSpotTier))
        (else
          (Variables|Items|SetTagToLoot_ToUse (Variables|Items|GetTagToLoot_DefaultTier))))
      ;[1 1 1] Check for biome tags to see if we should change the loot tier groups
      (if (Networking|IsServer)
        (bind _asfort_game_state_athena (Utilities|Casting|CastToFortGameStateAthena (Game|GetGameState)))
        (bind _returnvalue (POI|GetPoiGridTagsforLocation _asfort_game_state_athena (Transformation|GetActorLocation)))
        (Variables|Items|SetLootTags _returnvalue)
        ;[1 1 1] Push Fishing Spot Away From Shore If  It's Too Close To It
        (bind (_outhit _returnvalue_2) (Collision|BoxTraceByChannel (Transformation|GetActorLocation) (+ (Transformation|GetActorLocation) "0, 0,1.000000") "250.000000,250.000000,15.000000" "0, 0, 0" "TraceTypeQuery10"))
        (if _returnvalue_2
          (Transformation|SetActorLocation (+ (* (Math|Vector|Normalize (Math|Vector|Normalize2D(Vector) (- (.Location (Collision|BreakHitResult _outhit)) (Transformation|GetActorLocation)))) 2e+02) (Transformation|GetActorLocation))))))))

(event Custom|SpawnItem
  (params (Target :type Actor%20Object%20Reference) (Hooked :type Boolean) (HookComponent :type Scene%20Component%20Object%20Reference) (SpawnLoc :type Vector) (Player :type Actor%20Object%20Reference) (ItemSpawnDelay :type Float%20%28double-precision%29) (CaughtWithFishingRod :type Boolean) (CaughtWithHappyGhost :type Boolean) (ItemUsedToFish :type Actor%20Object%20Reference) (Projectile :type Object%20Reference) (Required%20Tags :type Gameplay%20Tag%20Container%20Structure))
  (bind _returnvalue (Fortnite|GetFortPlayerControllerfromActor Target))
  (bind _returnvalue_2 (Quests|GetQuestManager _returnvalue "Invalid"))
  (bind _returnvalue_3 (Math|Random|RandomFloatinRange 0.0 1.0))
  (bind _fishsizecurve (Variables|Default|GetFishSizeCurve))
  (bind _returnvalue_4 (Math|Curves|GetFloatValue _fishsizecurve _returnvalue_3))
  (bind _returnvalue_5 (Fortnite|GetFortPlayerControllerfromActor Target))
  (bind _spawneditem (Variables|Default|GetSpawnedItem))
  (bind _fishsizemap_itemtags (Variables|Default|GetFishSizeMap_ItemTags))
  ;[1 1 1] Spawn Items
  (Variables|Default|SetProjectile Projectile)
  (Development|PrintString "FlopperSpawnItems" (Variables|Default|GetDebug) (Variables|Default|GetDebug))
  (if (> ItemSpawnDelay 0)
    (Utilities|FlowControl|Delay ItemSpawnDelay) ; Used to delay the item spawn so that the water does not push the item around when it spawns
  )
  (if (Networking|IsServer)
    (if (not (Variables|Default|GetMaxUses))
      (bind _output_get (CallFunction|GetLootTier false (GameplayTags|Equal(GameplayTagContainer) Required%20Tags (GameplayTags|MakeGameplayTagContainerfromArray (Utilities|Array|MakeArray (Variables|Default|GetHighTierFishingRod)))) Player))
      (bind (_outloottodrop _returnvalue_6) (Loot|PickLootDrops _output_get 1 -1 "(GameplayTags=)"))
      (if _returnvalue_6
        (Development|PrintString (Utilities|String|Append "Fish Size:" (Utilities|String|ToString(Float) (Math|Float|Lerp (ScalableFloat|GetValueatLevel (CallFunction|FindFishSize _fishsizemap_itemtags (Class|FortItemFunctionLibrary|BreakFortItemEntry (Utilities|Array|Get(acopy) _outloottodrop)))) (ScalableFloat|GetValueatLevel (CallFunction|FindFishSize _fishsizemap_itemtags (Class|FortItemFunctionLibrary|BreakFortItemEntry (Utilities|Array|Get(acopy) _outloottodrop))) 1.0) _returnvalue_4))) true true "(R=0.676456,G=1.000000,B=0.000000,A=1.000000)" 7.0)
        (Development|PrintString (Utilities|String|ToString(Float) (+ (* (- (ScalableFloat|GetValueatLevel (CallFunction|FindFishSize _fishsizemap_itemtags (Class|FortItemFunctionLibrary|BreakFortItemEntry (Utilities|Array|Get(acopy) _outloottodrop))) 1.0) (ScalableFloat|GetValueatLevel (CallFunction|FindFishSize _fishsizemap_itemtags (Class|FortItemFunctionLibrary|BreakFortItemEntry (Utilities|Array|Get(acopy) _outloottodrop))))) (ScalableFloat|GetValueatLevel (Variables|Default|GetMiniGameFishSizePercentileThreshold))) (ScalableFloat|GetValueatLevel (CallFunction|FindFishSize _fishsizemap_itemtags (Class|FortItemFunctionLibrary|BreakFortItemEntry (Utilities|Array|Get(acopy) _outloottodrop)))))) true true "(R=1.000000,G=0.000000,B=0.134318,A=1.000000)" 7.0)
        (Default|CallOnFishingComplete)
        (CallFunction|ClearSavedCollectionData) ; make sure this is clear before we set it in Update Collection Component
        (for _array_element _outloottodrop
          (bind (_array_element_itemdefinition _array_element_itemguid _array_element_count) (Class|FortItemFunctionLibrary|BreakFortItemEntry _array_element))
          (Development|PrintString (Utilities|String|Append "Item To Spawn:" (Utilities|GetDisplayName _array_element_itemdefinition)) (Variables|Default|GetDebug) (Variables|Default|GetDebug))
          (bind _asfort_world_item_definition (Utilities|Casting|CastToFortWorldItemDefinition _array_element_itemdefinition))
          (bind _returnvalue_7 (Inventory|SpawnItemVariantPickupinWorld (Utilities|Struct|MakeSpawnItemVariantParams _asfort_world_item_definition Required%20Tags (Math|Float|Ceil (* _array_element_count (CallFunction|GetQuantMultiplierOverride Player))) (select Hooked SpawnLoc (Transformation|GetActorLocation)) (Transformation|GetActorForwardVector) 0 false false false -1 "Fishing" "Unset" 0 0 false)))
          ;;[0.2681 0.4323 1] Apply mod sets to weapons spawned
          (bind _asfort_weapon_item_definition (Utilities|Casting|CastToFortWeaponItemDefinition _asfort_world_item_definition)
            (:then
              (bind _returnvalue_8 (Class|FortWeaponModFunctionLibrary|GetRandomWeaponModSetData _asfort_weapon_item_definition))
              (Class|FortWeaponModFunctionLibrary|ApplyModSettoPickup _returnvalue_8 _returnvalue_7))
            (:CastFailed
              _))
          (bind _output_get_2 (Variables|Default|SetSpawnedItem _returnvalue_7)) ; Use this to track the spawned item. We only spawn one item so we can use this even though it is a for each
          (Default|CallOnItemSpawned _output_get_2)
          (CallFunction|UpdateCollectionsComponent (.Found (CallFunction|FindFishSize _fishsizemap_itemtags (Class|FortItemFunctionLibrary|BreakFortItemEntry (Utilities|Array|Get(acopy) _outloottodrop)))) Player _array_element_itemdefinition _output_get_2 (Math|Float|Lerp (ScalableFloat|GetValueatLevel (CallFunction|FindFishSize _fishsizemap_itemtags (Class|FortItemFunctionLibrary|BreakFortItemEntry (Utilities|Array|Get(acopy) _outloottodrop)))) (ScalableFloat|GetValueatLevel (CallFunction|FindFishSize _fishsizemap_itemtags (Class|FortItemFunctionLibrary|BreakFortItemEntry (Utilities|Array|Get(acopy) _outloottodrop))) 1.0) _returnvalue_4))
          (if Hooked
            (Utilities|IsValid HookComponent
              (:Is%20Valid
                ;;[1 1 1] Stop simulating the projectile movement component so that attaches properly
                (Game|Components|ProjectileMovement|StopSimulating (Utilities|Array|Get(acopy) (Actor|GetComponentsByClass _output_get_2 "/Script/Engine.ProjectileMovementComponent")) (Collision|MakeHitResult false))
                (Transformation|AttachActorToComponent _output_get_2 HookComponent "None" "SnapToTarget" "KeepRelative" "KeepWorld" true))
              (:Is%20Not%20Valid
                _))
            (Utilities|FlowControl|Sequence
              (:then_0
                ;[0.1374 1 0.9604 0.169] Quest hooks
                (Networking|SwitchHasAuthority
                  (:Authority
                    (if (not (Variables|Default|GetSpawnedFreeFish))
                      (Quest|SendComplexCustomStatEvent _returnvalue_2 Target (select (Variables|Default|GetIsWorldSpawner) (Variables|Default|GetQuestTag_WorldSpawn) (Variables|Default|GetQuestTag_DefaultSpawn)) (GameplayTags|GetOwnedGameplayTags _array_element_itemdefinition :!class /Script/GameplayTags.GameplayTagAssetInterface) 1 :!label _label)
                      (else
                        (bind _returnvalue_9 (GameplayTags|HasMatchingGameplayTag(Message) Target "(TagName=\"Athena.Quest.Muster.JellyAngler\")"))
                        (if _returnvalue_9
                          (goto _label))))
                    (if CaughtWithFishingRod
                      (Quest|SendComplexCustomStatEvent _returnvalue_2 Target "(GameplayTags=)" (CallFunction|GetTargetTagforRod ItemUsedToFish) 1)))))
              (:then_1
                ;[0.2681 0.4323 1] Call OnInteractionEventOccured on the playerstate
                (Networking|SwitchHasAuthority
                  (:Authority
                    (Utilities|IsValid _returnvalue_5
                      (:Is%20Valid
                        (bind _asfort_player_state_athena (Utilities|Casting|CastToFortPlayerStateAthena (Variables|/Script/Engine.Controller|GetPlayerState _returnvalue_5)))
                        (Class|FortPlayerStateAthena|OnInteractionEventOccurred _asfort_player_state_athena (GameplayTags|GetOwnedGameplayTags :!class /Script/GameplayTags.GameplayTagAssetInterface))))))))))
        (CallFunction|HandleUseTracking)
        (CallFunction|SomethingCaught (Variables|Default|GetSpawnedItem) Player Required%20Tags HookComponent)
        ;[1 1 1] End Fishing Event if caught from fishing rod - used save values from collection update call
        (if (or CaughtWithFishingRod CaughtWithHappyGhost)
          (bind _asfort_player_pawn (Utilities|Casting|CastToFortPlayerPawn Player)
            (:then
              (Fortnite|Fishing|FireEventEndFishingSessionwithData _asfort_player_pawn (Utilities|Struct|MakePlayerFishingTelemetryData _spawneditem ItemUsedToFish (Variables|Default|GetIsWorldSpawner) (Variables|CollectionAnalytics|GetNew%20Best%20Fish) (Variables|CollectionAnalytics|GetFish%20Variant%20String) (Variables|CollectionAnalytics|GetFishPoolTag) (Variables|CollectionAnalytics|GetFish%20Length)))
              (if (GameplayTags|HasMatchingGameplayTag (Class|FortItemFunctionLibrary|BreakFortItemEntry (Pickup|GetItemEntry _spawneditem 0)) "(TagName=\"Item.Weapon.Ranged\")")
                (Ability|SendGameplayEventtoActor _asfort_player_pawn "(TagName=\"Event.Quest.FishedUpWeapon\")" (Utilities|Struct|MakeGameplayEventData "(TagName=\"Event.Quest.FishedUpWeapon\")" 0 _spawneditem))
                (else
                  _)))
            (:CastFailed
              ;[1 1 1] Play Stinger if item was spawned from World Spawner
              (if (Variables|Default|GetIsWorldSpawner) :!label _label_2
                (Networking|FlushNetDormancy)
                (CallFunction|PlaySpawnItemHightier (Transformation|GetActorLocation Player))
                (else
                  _))))
          (CallFunction|SendQuestFishingEvent Target))
        (goto _label_2)))))

(event Custom|FishingComplete
  (bind _spawneditem (Variables|Default|GetSpawnedItem))
  ;[1 1 1] Spawn Items
  (Default|CallOnFishingComplete)
  (CallFunction|ClearSavedCollectionData) ; make sure this is clear before we set it in Update Collection Component
  (for _array_element (stale Custom|SpawnItem _outloottodrop)
    (bind (_array_element_itemdefinition _array_element_itemguid _array_element_count) (Class|FortItemFunctionLibrary|BreakFortItemEntry _array_element))
    (Development|PrintString (Utilities|String|Append "Item To Spawn:" (Utilities|GetDisplayName _array_element_itemdefinition)) (Variables|Default|GetDebug) (Variables|Default|GetDebug))
    (bind _asfort_world_item_definition (Utilities|Casting|CastToFortWorldItemDefinition _array_element_itemdefinition))
    (bind _returnvalue (Inventory|SpawnItemVariantPickupinWorld (Utilities|Struct|MakeSpawnItemVariantParams _asfort_world_item_definition (stale Custom|SpawnItem Required%20Tags) (Math|Float|Ceil (* _array_element_count (CallFunction|GetQuantMultiplierOverride (stale Custom|SpawnItem Player)))) (select (stale Custom|SpawnItem Hooked) (stale Custom|SpawnItem SpawnLoc) (Transformation|GetActorLocation)) (Transformation|GetActorForwardVector) 0 false false false -1 "Fishing" "Unset" 0 0 false)))
    ;;[0.2681 0.4323 1] Apply mod sets to weapons spawned
    (bind _asfort_weapon_item_definition (Utilities|Casting|CastToFortWeaponItemDefinition _asfort_world_item_definition)
      (:then
        (bind _returnvalue_2 (Class|FortWeaponModFunctionLibrary|GetRandomWeaponModSetData _asfort_weapon_item_definition))
        (Class|FortWeaponModFunctionLibrary|ApplyModSettoPickup _returnvalue_2 _returnvalue))
      (:CastFailed
        _))
    (bind _output_get (Variables|Default|SetSpawnedItem _returnvalue)) ; Use this to track the spawned item. We only spawn one item so we can use this even though it is a for each
    (Default|CallOnItemSpawned _output_get)
    (CallFunction|UpdateCollectionsComponent (.Found (CallFunction|FindFishSize (Variables|Default|GetFishSizeMap_ItemTags) (Class|FortItemFunctionLibrary|BreakFortItemEntry (Utilities|Array|Get(acopy) (stale Custom|SpawnItem _outloottodrop))))) (stale Custom|SpawnItem Player) _array_element_itemdefinition _output_get (Math|Float|Lerp (ScalableFloat|GetValueatLevel (CallFunction|FindFishSize (Variables|Default|GetFishSizeMap_ItemTags) (Class|FortItemFunctionLibrary|BreakFortItemEntry (Utilities|Array|Get(acopy) (stale Custom|SpawnItem _outloottodrop))))) (ScalableFloat|GetValueatLevel (CallFunction|FindFishSize (Variables|Default|GetFishSizeMap_ItemTags) (Class|FortItemFunctionLibrary|BreakFortItemEntry (Utilities|Array|Get(acopy) (stale Custom|SpawnItem _outloottodrop)))) 1.0) (Math|Curves|GetFloatValue (Variables|Default|GetFishSizeCurve) (Math|Random|RandomFloatinRange 0.0 1.0))))
    (if (stale Custom|SpawnItem Hooked)
      (Utilities|IsValid (stale Custom|SpawnItem HookComponent)
        (:Is%20Valid
          ;;[1 1 1] Stop simulating the projectile movement component so that attaches properly
          (Game|Components|ProjectileMovement|StopSimulating (Utilities|Array|Get(acopy) (Actor|GetComponentsByClass _output_get "/Script/Engine.ProjectileMovementComponent")) (Collision|MakeHitResult false))
          (Transformation|AttachActorToComponent _output_get (stale Custom|SpawnItem HookComponent) "None" "SnapToTarget" "KeepRelative" "KeepWorld" true))
        (:Is%20Not%20Valid
          _))
      (Utilities|FlowControl|Sequence
        (:then_0
          ;[0.1374 1 0.9604 0.169] Quest hooks
          (Networking|SwitchHasAuthority
            (:Authority
              (if (not (Variables|Default|GetSpawnedFreeFish))
                (Quest|SendComplexCustomStatEvent (Quests|GetQuestManager (Fortnite|GetFortPlayerControllerfromActor (stale Custom|SpawnItem Target)) "Invalid") (stale Custom|SpawnItem Target) (select (Variables|Default|GetIsWorldSpawner) (Variables|Default|GetQuestTag_WorldSpawn) (Variables|Default|GetQuestTag_DefaultSpawn)) (GameplayTags|GetOwnedGameplayTags _array_element_itemdefinition :!class /Script/GameplayTags.GameplayTagAssetInterface) 1 :!label _label)
                (else
                  (bind _returnvalue_3 (GameplayTags|HasMatchingGameplayTag(Message) (stale Custom|SpawnItem Target) "(TagName=\"Athena.Quest.Muster.JellyAngler\")"))
                  (if _returnvalue_3
                    (goto _label))))
              (if (stale Custom|SpawnItem CaughtWithFishingRod)
                (Quest|SendComplexCustomStatEvent (Quests|GetQuestManager (Fortnite|GetFortPlayerControllerfromActor (stale Custom|SpawnItem Target)) "Invalid") (stale Custom|SpawnItem Target) "(GameplayTags=)" (CallFunction|GetTargetTagforRod (stale Custom|SpawnItem ItemUsedToFish)) 1)))))
        (:then_1
          ;[0.2681 0.4323 1] Call OnInteractionEventOccured on the playerstate
          (Networking|SwitchHasAuthority
            (:Authority
              (Utilities|IsValid (Fortnite|GetFortPlayerControllerfromActor (stale Custom|SpawnItem Target))
                (:Is%20Valid
                  (bind _asfort_player_state_athena (Utilities|Casting|CastToFortPlayerStateAthena (Variables|/Script/Engine.Controller|GetPlayerState (Fortnite|GetFortPlayerControllerfromActor (stale Custom|SpawnItem Target)))))
                  (Class|FortPlayerStateAthena|OnInteractionEventOccurred _asfort_player_state_athena (GameplayTags|GetOwnedGameplayTags :!class /Script/GameplayTags.GameplayTagAssetInterface))))))))))
  (CallFunction|HandleUseTracking)
  (CallFunction|SomethingCaught (Variables|Default|GetSpawnedItem) (stale Custom|SpawnItem Player) (stale Custom|SpawnItem Required%20Tags) (stale Custom|SpawnItem HookComponent))
  ;[1 1 1] End Fishing Event if caught from fishing rod - used save values from collection update call
  (if (or (stale Custom|SpawnItem CaughtWithFishingRod) (stale Custom|SpawnItem CaughtWithHappyGhost))
    (bind _asfort_player_pawn (Utilities|Casting|CastToFortPlayerPawn (stale Custom|SpawnItem Player))
      (:then
        (Fortnite|Fishing|FireEventEndFishingSessionwithData _asfort_player_pawn (Utilities|Struct|MakePlayerFishingTelemetryData _spawneditem (stale Custom|SpawnItem ItemUsedToFish) (Variables|Default|GetIsWorldSpawner) (Variables|CollectionAnalytics|GetNew%20Best%20Fish) (Variables|CollectionAnalytics|GetFish%20Variant%20String) (Variables|CollectionAnalytics|GetFishPoolTag) (Variables|CollectionAnalytics|GetFish%20Length)))
        (if (GameplayTags|HasMatchingGameplayTag (Class|FortItemFunctionLibrary|BreakFortItemEntry (Pickup|GetItemEntry _spawneditem 0)) "(TagName=\"Item.Weapon.Ranged\")")
          (Ability|SendGameplayEventtoActor _asfort_player_pawn "(TagName=\"Event.Quest.FishedUpWeapon\")" (Utilities|Struct|MakeGameplayEventData "(TagName=\"Event.Quest.FishedUpWeapon\")" 0 _spawneditem))
          (else
            _)))
      (:CastFailed
        ;[1 1 1] Play Stinger if item was spawned from World Spawner
        (if (Variables|Default|GetIsWorldSpawner) :!label _label_2
          (Networking|FlushNetDormancy)
          (CallFunction|PlaySpawnItemHightier (Transformation|GetActorLocation (stale Custom|SpawnItem Player)))
          (else
            _))))
    (CallFunction|SendQuestFishingEvent (stale Custom|SpawnItem Target)))
  (goto _label_2))

(event Custom|SomethingCaught
  (params (ItemSpawned :type Fort%20Pickup%20Object%20Reference) (Player :type Actor%20Object%20Reference) (RequiredTags :type Gameplay%20Tag%20Container%20Structure) (HookComponent :type Scene%20Component%20Object%20Reference)))

(event Custom|PlaySpawnItemHightier
  (params (Location :type Vector))
  ;[1 1 1] Play Stinger if item was spawned from World Spawner
  (Audio|PlaySoundatLocation (Variables|Default|GetSound_HighTierPoolSpawn) Location))

(event Custom|UpdateCollectionsComponent
  (params (bCaughtFish :type Boolean) (Player :type Actor%20Object%20Reference) (ItemDefinition :type Item%20Definition%20Base%20Object%20Reference) (Pickup :type Fort%20Pickup%20Object%20Reference) (Length :type Float%20%28double-precision%29))
  ;[1 1 1] Attempt to add this caught item to the collection.  If this is a fish, we save information for analytics about the variant type, size, and if it's the best the player has caught so far
  (if bCaughtFish
    (bind _asfort_pawn (Utilities|Casting|CastToFortPawn Player))
    (bind _returnvalue (Class|FortControllerComponentCollections|GetCollectionsComponent _asfort_pawn)
      (:IsValid
        (Utilities|IsValid Pickup
          (:Is%20Valid
            (Utilities|IsValid ItemDefinition
              (:Is%20Valid
                (bind _output_get (Variables|CollectionAnalytics|SetFish%20Length Length))
                (bind (_outfishcollectiontag _outanalyticstag _outnewseenstate) (Class|FortControllerComponentCollections|OnFishCollected _returnvalue (Pickup|GetItemEntry Pickup 0) (Utilities|Struct|MakeFortMcpCollectedFishProperties 3.0 _output_get))) ; Returns the Collection Tag (which is the variant identifier) and an analytics tag, that currently isn't setup in data - we might want to switch to that in the future, but for now the collection tag properly ties the variant type to the collection, it seems like the safest data
                (Variables|CollectionAnalytics|SetFish%20Variant%20String (Utilities|String|ToString(Name) (GameplayTags|GetTagName _outfishcollectiontag)))
                (switch Utilities|FlowControl|Switch|SwitchonEFortCollectedState _outnewseenstate
                  (:NewlyCollected
                    _)
                  (:NewBest
                    _)
                  (:NewRecord
                    _))
                (Variables|CollectionAnalytics|SetNew%20Best%20Fish true)))))))))

(event Custom|Clear%20Saved%20Collection%20Data
  ;[1 1 1] Make sure we don't have saved data for non-fish catches
  (Variables|CollectionAnalytics|SetFish%20Variant%20String "")
  (Variables|CollectionAnalytics|SetFish%20Length 0.0)
  (Variables|CollectionAnalytics|SetNew%20Best%20Fish false))

(fn BlueprintCanInteract
  (params (InteractingPawn :type Fort%20Pawn%20Object%20Reference) (InteractionBeingAttempted :type EInteractionBeingAttempted%20Enum :default "FirstInteraction") (InteractionType :type TInteractionType%20Enum :default "IT_NoInteraction"))
  (return false))

(fn CanRespawn
  (return (> (ScalableFloat|GetValueatLevel (Variables|Rows|GetRow_RespawnDelayMin)) 0.0)))

(fn CanGrantQuestProgress
  ;[0.03429 0.3069 0.7899] To be overwritten by children
  (return true))

(fn DestroyFlopperSpawn
  ;[0.2086 0.3763 1 0.25] Hiding the flopper spawn and destroying it after a short delay. Encapsulated in a function so it can be overridden by children.
  (Rendering|SetActorHiddenInGame true)
  (Collision|SetActorEnableCollision false)
  (Actor|SetLifeSpan (Variables|Default|GetAfterHideLifeSpan)))

(fn FindFishSize
  (params (FishSizeMap :type Map%20of%20Gameplay%20Tag%20Structures%20to%20Scalable%20Float%20Structures) (ItemDef :type Item%20Definition%20Base%20Object%20Reference))
  (bind (_value _returnvalue) (Utilities|Map|Find (Variables|Default|GetFishSizeMap_ItemTags) (LocalVariables|GetLocalFishSize)))
  (bind _keys (Utilities|Map|Keys FishSizeMap))
  (bind _array_element (Utilities|Array|ForEachLoopwithBreak _keys :!label _label)
    (:LoopBody
      (if (GameplayTags|HasTag (GameplayTags|GetOwnedGameplayTags ItemDef :!class /Script/GameplayTags.GameplayTagAssetInterface) _array_element true)
        (LocalVariables|SetLocalFishSize _array_element)
        (goto _label Break)))
    (:Completed
      (return _value _returnvalue))) ; Break when one is found
)

(fn GetLootTier
  (params (IsExplosion :type Boolean) (IsProFishingRod :type Boolean) (Player :type Actor%20Object%20Reference))
  (bind (_success _itemstodrop) (CallFunction|GetLootOverride Player))
  (bind _tagtoloot_touse (Variables|Items|GetTagToLoot_ToUse))
  ;[1 1 1] Check if the spawner has any tags for Location Type if so change the loot.
  (if _success
    (return _itemstodrop)
    (elif IsProFishingRod
      (Variables|Items|SetTagToLoot_ToUse (Variables|Items|GetTagToLoot_HighTier_FishingRod))))
  (bind _keys (Utilities|Map|Keys _tagtoloot_touse))
  (for _array_element _keys
    (if (and (GameplayTags|HasTag (Variables|Items|GetLootTags) _array_element true) (.ReturnValue (Utilities|Map|Find _tagtoloot_touse _array_element)))
      (bind _output_get (Variables|Items|SetItemsToDrop (Utilities|Map|Find _tagtoloot_touse _array_element)))
      (else
        (Variables|Items|SetItemsToDrop (Utilities|Map|Find (Variables|Items|GetTagToLoot_ToUse) (Variables|Default|GetDefaultLootTag)))))
    (Development|PrintString (Utilities|String|ToString(Name) _output_get) (Variables|Default|GetDebug) (Variables|Default|GetDebug) "(R=1.000000,G=0.407895,B=0.000000,A=1.000000)"))
  (return (Variables|Items|GetItemsToDrop)))

(fn GetBonusChanceForBigBob
  (if (Variables|Default|GetIsWorldSpawner)
    (return (Math|Random|RandomFloatinRange (ScalableFloat|GetValueatLevel (Variables|Rows|GetRow_AddedBobChanceOnOverlapMin)) (ScalableFloat|GetValueatLevel (Variables|Rows|GetRow_AddedBobChanceOnOverlapMax))))
    (else
      (return 0.0))))

(fn GetLootOverride
  (params (Player :type Actor%20Object%20Reference))
  (bind _returnvalue (Actor|GetComponentbyClass Player "/Game/Athena/Items/EnvironmentalItems/FlopperSpawn/Component_FlopperSpawnOverride.Component_FlopperSpawnOverride_C"))
  (Networking|SwitchHasAuthority
    (:Authority
      (Utilities|IsValid Player
        (:Is%20Valid
          (Utilities|IsValid _returnvalue
            (:Is%20Valid
              (if (Variables|/Game/Athena/Items/EnvironmentalItems/FlopperSpawn/Component_FlopperSpawnOverride.Component_FlopperSpawnOverride_C|GetOverridesLoot? _returnvalue)
                (return true (Variables|/Game/Athena/Items/EnvironmentalItems/FlopperSpawn/Component_FlopperSpawnOverride.Component_FlopperSpawnOverride_C|GetOverrideLootTier _returnvalue))
                (else
                  _)))
            (:Is%20Not%20Valid
              _))
          (return false "None"))
        (:Is%20Not%20Valid
          (return false "None"))))))

(fn GetQuantMultiplierOverride
  (params (Player :type Actor%20Object%20Reference))
  (bind _returnvalue (Actor|GetComponentbyClass Player "/Game/Athena/Items/EnvironmentalItems/FlopperSpawn/Component_FlopperSpawnOverride.Component_FlopperSpawnOverride_C"))
  (Utilities|IsValid _returnvalue
    (:Is%20Valid
      (return (Variables|/Game/Athena/Items/EnvironmentalItems/FlopperSpawn/Component_FlopperSpawnOverride.Component_FlopperSpawnOverride_C|GetQuantityMultiplier _returnvalue)))
    (:Is%20Not%20Valid
      (return 1.0))))

(fn GetTargetTagForRod
  (params (Actor :type Actor%20Object%20Reference))
  ;[0 1 0.992 0.424] If the weapon that's fishing has the appropriate tag for a high tier fishing rod, then return the target tag for the high tier fishing rod, else return a normal fishing rod result.
  (bind _asfort_weapon (Utilities|Casting|CastToFortWeapon Actor))
  (bind _asfort_world_item (Utilities|Casting|CastToFortWorldItem (FortWeapon|GetInventoryItem _asfort_weapon)))
  (return (select (GameplayTags|HasTag (GameplayTags|GetOwnedGameplayTags _asfort_world_item :!class /Script/FortniteGame.FortWorldItem) (Variables|Default|GetHighTierFishingRod) true) "(GameplayTags=((TagName=\"Athena.Quests.FishingCatchItem.FishingRod.High\")))" "(GameplayTags=((TagName=\"Athena.Quests.FishingCatchItem.FishingRod\")))")))

(fn HandleUseTracking
  (bind _result (Math|Integer|DecrementInt (Variables|Default|GetUses))
    (:%20%20
      (if (<= _result 0)
        (Variables|Default|SetMaxUses true)
        (Networking|FlushNetDormancy)
        (if (CallFunction|CanRespawn)
          (CallFunction|SetRespawnDelay)
          (CallFunction|HideandDelayforRespawn (Variables|Default|GetRespawnDelay))
          (else
            (CallFunction|HideandKill)))))))

(fn SendQuestFishingEvent
  (params (Actor :type Actor%20Object%20Reference))
  (bind _asfort_player_pawn_athena (Utilities|Casting|CastToFortPlayerPawnAthena Actor))
  (Ability|SendGameplayEventtoActor _asfort_player_pawn_athena "(TagName=\"Event.Quest.SpawnQuestActor.Fishing\")" (Utilities|Struct|MakeGameplayEventData "(TagName=\"Event.Quest.SpawnQuestActor.Fishing\")" _asfort_player_pawn_athena self (Variables|Default|GetProjectile) (Class|FortItemFunctionLibrary|BreakFortItemEntry (Pickup|GetItemEntry (Variables|Default|GetSpawnedItem) 0)) "()" "(GameplayTags=)" (GameplayTags|GetOwnedGameplayTags self :!class /Script/GameplayTags.GameplayTagAssetInterface))))

(fn SetBalanceValues
  (Variables|Default|SetUses (Math|Random|RandomIntegerinRange (Math|Float|Truncate (ScalableFloat|GetValueatLevel (Variables|Rows|GetRow_MinUses))) (Math|Float|Truncate (ScalableFloat|GetValueatLevel (Variables|Rows|GetRow_MaxUses))))))

(fn SetRespawnDelay
  (Variables|Default|SetRespawnDelay (Math|Random|RandomFloatinRange (ScalableFloat|GetValueatLevel (Variables|Rows|GetRow_RespawnDelayMin)) (ScalableFloat|GetValueatLevel (Variables|Rows|GetRow_RespawnDelayMax)))))

(fn ShouldKill
  (if (Variables|Default|GetAlreadySpawnChecked)
    _
    (elif (Math|Random|RandomBoolwithWeight (Math|Random|RandomFloatinRange (ScalableFloat|GetValueatLevel (Variables|Rows|GetRow_SpawnChanceMin)) (ScalableFloat|GetValueatLevel (Variables|Rows|GetRow_SpawnChanceMax))))
      _
      (else
        (Actor|SetLifeSpan (Math|Random|RandomFloatinRange 0.1 0.5))))))

(fn ConstructionScript
  (if (not (Networking|IsServer))
    (Variables|/Script/Engine.PrimitiveComponent|SetbGenerateOverlapEvents false (Variables|Default|GetOverlapVolume))))
