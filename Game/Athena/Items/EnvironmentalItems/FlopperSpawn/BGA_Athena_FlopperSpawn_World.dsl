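; World fishing spot (flopper spawn) that shows ambient fish effects on water and yields loot when fished or blasted.
; On begin play or reinitialize, the server can trace for water, store the body and snap to its surface.
; OnRep_FortWaterBody aligns effects and feeds water data to materials, while cosmetic timers spawn jumping fish.
; Nearby explosion damage triggers one server loot burst with quest and analytics reporting, then hides and kills the spot.
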(event Custom|Reinitialize%20Flopper
  (bind (_returnvalue_x _returnvalue_y) (Math|Vector|BreakVector (Transformation|GetActorLocation)))
  (bind _sm_boilingwave (Variables|Default|GetSM_BoilingWave))
  (if (or (Variables|Default|GetSpawnInstance?) (not (Variables|Default|GetPlayFishFX)))
    (Components|Activation|Deactivate (Variables|Default|GetNS_Flopper_Spot))
    (else
      (Components|Activation|Activate (Variables|Default|GetNS_Flopper_Spot) false)))
  (Utilities|FlowControl|Sequence
    (:then_0
      (Networking|SwitchHasAuthority
        (:Authority
          (if (Variables|Default|GetTestForWaterOnStart)
            ;[1 1 1] Save out the Water Body
            (bind (_outhit _returnvalue) (Collision|LineTraceByChannel (+ (Transformation|GetActorLocation) "0, 0,5.000000") (- (Transformation|GetActorLocation) (Variables|Default|GetTraceOffsetAmount)) "TraceTypeQuery12" true))
            (if _returnvalue
              (Variables|Default|SetWaterNormalHit (.Normal (Collision|BreakHitResult _outhit)))
              (bind _asfort_water_body_actor (Utilities|Casting|CastToFortWaterBodyActor (.HitActor (Collision|BreakHitResult _outhit))))
              (bind _output_get (Variables|Default|SetFortWaterBody _asfort_water_body_actor))
              (bind _outwatersurfaceinfo (WaterBody|GetWaterSurfaceInfoStruct _output_get (Transformation|GetActorLocation) true false false))
              (bind (_waterplanelocation _waterplanenormal _watersurfacelocation) (Utilities|Struct|BreakWaterSurfaceInfo _outwatersurfaceinfo))
              (bind (_watersurfacelocation_x _watersurfacelocation_y _watersurfacelocation_z) (Math|Vector|BreakVector _watersurfacelocation))
              (Transformation|SetActorLocation 0 (Math|Vector|MakeVector _returnvalue_x _returnvalue_y _watersurfacelocation_z) false false)
              ;[1 1 1] Destroy self if deeply overlapping terrain
              (bind (_outhit_2 _returnvalue_2) (Collision|BoxTraceByChannel (Transformation|GetActorLocation) (+ (Transformation|GetActorLocation) "0, 0,1.000000") "20.000000,20.000000,20.000000" "0, 0, 0" "TraceTypeQuery10"))
              (if _returnvalue_2
                _))
            (Actor|DestroyActor)))))
    (:then_1
      ;[1 1 1] Spawn Jumping fish
      (Networking|CanExecuteCosmeticEvents
        (:True
          (if (Variables|Default|GetPlayFishFX)
            (Utilities|Time|SetTimerbyEvent (delegate Custom|SpawnJumpingFish) (CallFunction|RandomTimeInterval))))))
    (:then_2
      (Networking|CanExecuteCosmeticEvents
        (:True
          (Utilities|FlowControl|DelayUntilNextTick)
          (Utilities|FlowControl|DelayUntilNextTick)
          (Utilities|FlowControl|DelayUntilNextTick)
          (bind _returnvalue_3 (Settings|GetGameUserSettings))
          (if (== (Settings|GetVisualEffectQuality _returnvalue_3) 0)
            (switch int (Math|Float|Round (Fortnite|GetNumericCVar "fx.FishingSpotDebuggingLevel"))
              (:1
                (Rendering|Material|SetScalarParameterValueonMaterials _sm_boilingwave "TmpDisableWaterDisplacement" 1.0))
              (:2
                (Rendering|Material|SetScalarParameterValueonMaterials _sm_boilingwave "TmpDisableWaterDisplacement" 1.0)
                (Rendering|Material|SetScalarParameterValueonMaterials _sm_boilingwave "TmpDisableGameTimeFade" 1.0)))))))))

(event EventBeginPlay
  (bind (_returnvalue_x _returnvalue_y) (Math|Vector|BreakVector (Transformation|GetActorLocation)))
  (bind _sm_boilingwave (Variables|Default|GetSM_BoilingWave))
  (|CallParentFunction)
  (if (or (Variables|Default|GetSpawnInstance?) (not (Variables|Default|GetPlayFishFX)))
    (Components|Activation|Deactivate (Variables|Default|GetNS_Flopper_Spot))
    (else
      (Components|Activation|Activate (Variables|Default|GetNS_Flopper_Spot) false)))
  (Utilities|FlowControl|Sequence
    (:then_0
      (Networking|SwitchHasAuthority
        (:Authority
          (if (Variables|Default|GetTestForWaterOnStart)
            ;[1 1 1] Save out the Water Body
            (bind (_outhit _returnvalue) (Collision|LineTraceByChannel (+ (Transformation|GetActorLocation) "0, 0,5.000000") (- (Transformation|GetActorLocation) (Variables|Default|GetTraceOffsetAmount)) "TraceTypeQuery12" true))
            (if _returnvalue
              (Variables|Default|SetWaterNormalHit (.Normal (Collision|BreakHitResult _outhit)))
              (bind _asfort_water_body_actor (Utilities|Casting|CastToFortWaterBodyActor (.HitActor (Collision|BreakHitResult _outhit))))
              (bind _output_get (Variables|Default|SetFortWaterBody _asfort_water_body_actor))
              (bind _outwatersurfaceinfo (WaterBody|GetWaterSurfaceInfoStruct _output_get (Transformation|GetActorLocation) true false false))
              (bind (_waterplanelocation _waterplanenormal _watersurfacelocation) (Utilities|Struct|BreakWaterSurfaceInfo _outwatersurfaceinfo))
              (bind (_watersurfacelocation_x _watersurfacelocation_y _watersurfacelocation_z) (Math|Vector|BreakVector _watersurfacelocation))
              (Transformation|SetActorLocation 0 (Math|Vector|MakeVector _returnvalue_x _returnvalue_y _watersurfacelocation_z) false false)
              ;[1 1 1] Destroy self if deeply overlapping terrain
              (bind (_outhit_2 _returnvalue_2) (Collision|BoxTraceByChannel (Transformation|GetActorLocation) (+ (Transformation|GetActorLocation) "0, 0,1.000000") "20.000000,20.000000,20.000000" "0, 0, 0" "TraceTypeQuery10"))
              (if _returnvalue_2
                _))
            (Actor|DestroyActor)))))
    (:then_1
      ;[1 1 1] Spawn Jumping fish
      (Networking|CanExecuteCosmeticEvents
        (:True
          (if (Variables|Default|GetPlayFishFX)
            (Utilities|Time|SetTimerbyEvent (delegate Custom|SpawnJumpingFish) (CallFunction|RandomTimeInterval))))))
    (:then_2
      (Networking|CanExecuteCosmeticEvents
        (:True
          (Utilities|FlowControl|DelayUntilNextTick)
          (Utilities|FlowControl|DelayUntilNextTick)
          (Utilities|FlowControl|DelayUntilNextTick)
          (bind _returnvalue_3 (Settings|GetGameUserSettings))
          (if (== (Settings|GetVisualEffectQuality _returnvalue_3) 0)
            (switch int (Math|Float|Round (Fortnite|GetNumericCVar "fx.FishingSpotDebuggingLevel"))
              (:1
                (Rendering|Material|SetScalarParameterValueonMaterials _sm_boilingwave "TmpDisableWaterDisplacement" 1.0))
              (:2
                (Rendering|Material|SetScalarParameterValueonMaterials _sm_boilingwave "TmpDisableWaterDisplacement" 1.0)
                (Rendering|Material|SetScalarParameterValueonMaterials _sm_boilingwave "TmpDisableGameTimeFade" 1.0)))))))))

(event Custom|SpawnJumpingFish
  ;[1 1 1] Spawn Jumping fish
  (if (Variables|Default|GetPlayFishFX)
    (if (Variables|Rendering|GetbHidden)
      (if (CallFunction|CanRespawn)
        (Utilities|Time|SetTimerbyEvent (EventDispatchers|CreateEvent :!event Custom|SpawnJumpingFish) (+ (CallFunction|RandomTimeInterval) 2.4e+02)) ; setup the next timer jump after we respawn - should be minutes
      )
      (else
        (bind (_outspawnlocation _outspawnrotation) (CallFunction|GetFlopperSpawnPositionandRotation (Variables|Default|GetOffsetDistance)))
        ;;[1 0.3849 0.3749] Niagara
        (bind _returnvalue (Niagara|SpawnSystematLocation (Variables|Default|GetNS_FX_Jump) _outspawnlocation _outspawnrotation))
        (Niagara|SetNiagaraVariable(LinearColor) _returnvalue "WaterColor" (Variables|Default|GetWaterBodyPrimaryColor))
        (Audio|PlaySoundatLocation (Variables|Default|GetSound_AmbientFishFlop) _outspawnlocation)
        (Utilities|Time|SetTimerbyEvent (EventDispatchers|CreateEvent :!event Custom|SpawnJumpingFish) (CallFunction|RandomTimeInterval)) ; setup the next short jump timer
      ))))

(event EventSpawnItem
  (params (Target :type Actor%20Object%20Reference) (Hooked :type Boolean) (HookComponent :type Scene%20Component%20Object%20Reference) (SpawnLoc :type Vector :default "0, 0, 0") (Player :type Actor%20Object%20Reference) (ItemSpawnDelay :type Float%20%28double-precision%29) (CaughtWithFishingRod :type Boolean) (CaughtWithHappyGhost :type Boolean) (ItemUsedToFish :type Actor%20Object%20Reference) (Projectile :type Object%20Reference) (Required%20Tags :type Gameplay%20Tag%20Container%20Structure))
  (|CallParentFunction Target Hooked HookComponent SpawnLoc Player ItemSpawnDelay CaughtWithFishingRod CaughtWithHappyGhost ItemUsedToFish Projectile Required%20Tags)
  (if (Variables|Default|GetSpawnInstance?)
    (CallFunction|HideandKill)))

(event EventOnDeathPlayEffects
  (params (Damage :type Float%20%28single-precision%29) (DamageTags :type Gameplay%20Tag%20Container%20Structure%20%28by%20ref%29) (Momentum :type Vector :default "0, 0, 0") (HitInfo :type Hit%20Result%20Structure%20%28by%20ref%29) (InstigatedBy :type Fort%20Pawn%20Object%20Reference) (DamageCauser :type Actor%20Object%20Reference) (EffectContext :type Gameplay%20Effect%20Context%20Handle%20Structure))
  ;[1 1 1] On Death Fade Sounds
  (Audio|Components|Audio|FadeOut (Variables|Default|GetAudio_FlopperLoop) 1.0 0.0 "Linear" :!class /Script/Engine.AudioComponent))

(event EventOnDamageServer
  (params (Damage :type Float%20%28single-precision%29) (DamageTags :type Gameplay%20Tag%20Container%20Structure%20%28by%20ref%29) (Momentum :type Vector :default "0, 0, 0") (HitInfo :type Hit%20Result%20Structure%20%28by%20ref%29) (InstigatedBy :type Controller%20Object%20Reference) (DamageCauser :type Actor%20Object%20Reference) (EffectContext :type Gameplay%20Effect%20Context%20Handle%20Structure))
  (bind (_bblockinghit _binitialoverlap _time _distance _location _impactpoint) (Collision|BreakHitResult HitInfo))
  ;[1 1 1] If we take damage from something and it it is close enough then spawn the explode items given we are immune to all damage but explosion.
  (if (< (Math|Vector|Distance(Vector) _impactpoint (Transformation|GetActorLocation self)) (Variables|Explosion|GetMinDistanceToExplosion))
    (CallFunction|SpawnfromExplosion InstigatedBy DamageCauser)))

(event Custom|SpawnFromExplosion
  (params (Instigator :type Actor%20Object%20Reference) (DamageCauser :type Actor%20Object%20Reference))
  (bind _returnvalue (Fortnite|GetFortPlayerControllerfromActor Instigator))
  (bind _returnvalue_2 (Transformation|GetActorLocation))
  (bind _explosionitemspawnoffset (Variables|Default|GetExplosionItemSpawnOffset))
  (bind _returnvalue_3 (+ _returnvalue_2 _explosionitemspawnoffset))
  ;[1 1 1] Spawn Items from the explosion loot pool
  (if (Networking|IsServer)
    (if (not (Variables|Default|GetMaxUses))
      (CallFunction|ExplosionQuestProgress Instigator DamageCauser)
      (Variables|Items|SetItemsToDrop "Loot_AthenaFlopper_Default")
      (Variables|Default|SetbHasWrittenAnalytic false)
      (for _ (range 1 (+ (Variables|Explosion|GetExplodeRolls) 1))
        (bind (_outloottodrop _returnvalue_4) (Loot|PickLootDrops (Variables|Items|GetItemsToDrop) 1 -1 "(GameplayTags=)"))
        (if _returnvalue_4
          (for _array_element _outloottodrop
            (bind (_array_element_itemdefinition _array_element_itemguid _array_element_count) (Class|FortItemFunctionLibrary|BreakFortItemEntry _array_element))
            (bind _asfort_world_item_definition (Utilities|Casting|CastToFortWorldItemDefinition _array_element_itemdefinition))
            (bind _returnvalue_5 (Inventory|SpawnPickupInWorld _asfort_world_item_definition _array_element_count _returnvalue_3 (Transformation|GetActorForwardVector) 0 true true false -1 "Other" "Unset" 0 false))
            (if (not (Variables|Default|GetbHasWrittenAnalytic))
              ;;[1 1 1] Write fishing analytic
              (Variables|Default|SetbHasWrittenAnalytic true)
              (Utilities|IsValid _returnvalue
                (:Is%20Valid
                  (bind _asfort_player_pawn (Utilities|Casting|CastToFortPlayerPawn (Pawn|GetControlledPawn _returnvalue))
                    (:then
                      (Fortnite|Fishing|FireEventBeginFishingSession _asfort_player_pawn)
                      (Fortnite|Fishing|FireEventEndFishingSessionwithData _asfort_player_pawn (Utilities|Struct|MakePlayerFishingTelemetryData _returnvalue_5 DamageCauser true false "" (Variables|CollectionAnalytics|GetFishPoolTag))))
                    (:CastFailed
                      _)))
                (:Is%20Not%20Valid
                  _)))
            (bind _asfort_weapon_item_definition (Utilities|Casting|CastToFortWeaponItemDefinition (Class|FortItemFunctionLibrary|BreakFortItemEntry (Pickup|GetItemEntry _returnvalue_5 0))))
            (Fortnite|Weapon|CreateTossAmmoPickupforWeaponItemDefinitionatLocation _asfort_weapon_item_definition (Variables|Default|GetT_FlopperSpawn) _returnvalue_3)
            (Quest|SendComplexCustomStatEvent (Quests|GetQuestManager _returnvalue "Invalid") 0 (Variables|Default|GetExplosionSpawnTag) (GameplayTags|GetOwnedGameplayTags _array_element_itemdefinition :!class /Script/GameplayTags.GameplayTagAssetInterface) 1))))
      (Variables|Default|SetMaxUses true)
      (CallFunction|HideandKill))))

(fn CanGrantQuestProgress)

(fn ExplosionQuestProgress
  (params (Instigator :type Actor%20Object%20Reference) (DamageCauser :type Actor%20Object%20Reference))
  ;[1 1 1] If instigator is player, report quest normally
  (bind _asfort_player_controller (Utilities|Casting|CastToFortPlayerController Instigator)
    (:then
      (bind _returnvalue (GameplayTags|GetOwnedGameplayTags _asfort_player_controller :!class /Script/GameplayTags.GameplayTagAssetInterface))
      (Quest|SendComplexCustomStatEvent (Quests|GetQuestManager (Fortnite|GetFortPlayerControllerfromActor _asfort_player_controller) "Invalid") 0 _returnvalue "(GameplayTags=((TagName=\"Athena.Quests.FishingCatchItem.Explosive\")))" 1)
      (CallFunction|SendQuestFishingEvent (Game|Player|GetPlayerPawn _asfort_player_controller)))
    (:CastFailed
      ;[1 1 1] https://jira.it.epicgames.net/browse/FORT-250707
      ; Making this check for the expected tag on the projectile instead of casting to the projectile and the BLGA that spawned it. So we don't reference these assets.
      ;
      ; Previous Comment: Handle quest progress if instigator is AutoLauncher_Explosive (BattleLab)
      (bind _returnvalue_2 (GameplayTags|HasAnyMatchingGameplayTags(Message) DamageCauser (Variables|Default|GetBattleLabProjectileTag)))
      (if _returnvalue_2
        (bind _returnvalue_3 (Fortnite|HeldObjects|GetHeldObjectsOwningPlayer (Actor|GetOwner DamageCauser)))
        (Utilities|IsValid _returnvalue_3
          (:Is%20Valid
            (Quest|SendComplexCustomStatEvent (Quests|GetQuestManager (Fortnite|GetFortPlayerControllerfromActor _returnvalue_3) "Invalid") 0 "(GameplayTags=)" "(GameplayTags=((TagName=\"Athena.Quests.FishingCatchItem.Explosive\")))" 1)))))))

(fn GetLootTier
  (params (IsExplosion :type Boolean) (IsProFishingRod :type Boolean) (Player :type Actor%20Object%20Reference))
  ;[1 1 1] Explosions just use the explosion tag, and call parent.
  (if IsExplosion
    (Variables|Items|SetTagToLoot_ToUse (Variables|Explosion|GetMAP_TagToLootExplode)))
  (bind _output_get (|CallParentFunction IsExplosion IsProFishingRod Player))
  (return _output_get))

(fn IsFishingEnabled
  ;[0.1771 0.1771 0.1771 0.25] Check if fishing is enabled in Creative Royale *Disabled*
  (if (Utilities|IsValid (Minigame|GetMinigame self :!class /Script/FortniteGame.MinigameBlueprintLibrary) :!class /Script/Engine.KismetSystemLibrary)
    (if (not (Math|Conversions|ToBoolean(Integer) (Math|Float|Truncate (ScalableFloat|GetValueatLevelwithContext (Variables|Default|GetEnabledRow)))))
      (return false)))
  (return true))

(fn OnRep_FortWaterBody
  (bind _sm_foamwave (Variables|Default|GetSM_FoamWave))
  (bind _sm_foamwave_2 (Variables|Default|GetSM_FoamWave))
  ;[1 1 1] Pass in water body information to the particle system
  (Networking|CanExecuteCosmeticEvents
    (:True
      (bind _fortwaterbody (Variables|Default|GetFortWaterBody)
        (:then
          (Utilities|FlowControl|Sequence
            (:then_0
              (bind _outwatersurfaceinfo (WaterBody|GetWaterSurfaceInfoStruct _fortwaterbody (Transformation|GetActorLocation) true false false))
              (Transformation|SetWorldTransform (Variables|Default|GetNS_Flopper_Spot) (Math|Transform|SelectTransform (Math|Transform|MakeTransform (+ (Variables|Default|GetWaterPlaneOffset) (Utilities|Struct|BreakWaterSurfaceInfo _outwatersurfaceinfo)) (Math|Rotator|MakeRotator (.roll (Math|Rotator|MakeRotfromZ (Variables|Default|GetWaterNormalHit))) (.pitch (Math|Rotator|MakeRotfromZ (Variables|Default|GetWaterNormalHit))))) (Math|Transform|MakeTransform (Transformation|GetWorldLocation (Variables|Default|GetNS_Flopper_Spot)) (Math|Rotator|MakeRotator (.roll (Math|Rotator|MakeRotfromZ (Variables|Default|GetWaterNormalHit))) (.pitch (Math|Rotator|MakeRotfromZ (Variables|Default|GetWaterNormalHit))))) (Variables|Default|GetUpdateFishLocationFromWater)) false false)
              (Rendering|Material|SetScalarParameterValueonMaterials _sm_foamwave "WaterBodyIndex" (Math|Conversions|ToFloat(Integer) (.WaterBodyIdx (Utilities|Struct|BreakWaterSurfaceInfo _outwatersurfaceinfo))))
              (Rendering|Material|SetScalarParameterValueonMaterials _sm_foamwave "WaterDepth" (.WaterDepth (Utilities|Struct|BreakWaterSurfaceInfo _outwatersurfaceinfo))))
            (:then_1
              (bind _output_get (Variables|Default|SetWaterBodyPrimaryColor (Variables|/Script/FortniteGame.FortWaterBodyActor|GetWaterBodyPrimaryColor _fortwaterbody)))
              (bind _returnvalue (Rendering|Material|CreateDynamicMaterialInstance _sm_foamwave_2 0 (Rendering|Material|GetMaterial _sm_foamwave_2 0) "None" :!class /Script/Engine.PrimitiveComponent))
              (bind _output_get_2 (Variables|Default|SetRippleMat _returnvalue))
              (Rendering|Material|SetVectorParameterValue _output_get_2 "WaterBodyPrimaryColor" _output_get :!class /Script/Engine.MaterialInstanceDynamic)
              (Rendering|Material|SetMaterial _sm_foamwave_2 0 (Variables|Default|GetRippleMat)))))
        (:else)))))

(fn RandomTimeInterval
  (return (Math|Random|RandomFloatinRange (- (Variables|Default|GetTime_Interval) (Variables|Default|GetTimeIntervalOffset)) (+ (Variables|Default|GetTime_Interval) (Variables|Default|GetTimeIntervalOffset)))))

(fn ConstructionScript
  (|CallParentFunction))
