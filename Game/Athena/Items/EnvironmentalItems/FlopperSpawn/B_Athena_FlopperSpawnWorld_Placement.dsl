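; Server-side placement actor that spawns Flopper fishing pool spawners across the world's water once playlist data is ready, or expires if disabled.
; Rolls candidate points along river, lake, island, ocean and custom splines, using random or spline-normal offsets.
; After a delay for ocean flood height, an EQS grid query supplies jittered ocean spawns.
; Staggered loops then spawn the pools, skipping water-body points well below ocean height.
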(event EventBeginPlay
  ;[1 1 1] Check if enabled, if not kill actor
  (if (Networking|IsServer)
    (Class|FortAsyncActionCurrentPlaylistReady|PlaylistDataReadyAsync
      (:OnReady
        (Utilities|Enabled
          (:true
            ;[1 1 1] Find Locations then start spawn
            (CallFunction|FindRiverNarrowLocations)
            (Default|FindRiverWideLocations)
            (CallFunction|FindLakeLocations)
            (CallFunction|FindIslandLocations)
            (CallFunction|FindOceanLocations)
            (Default|FindCustomSplineLocations)
            (Development|PrintString (Utilities|String|ToString(Float) (Water|GetOceanTotalHeight (WorldSubsystems|GetWaterSubsystem))) true true "(R=0.000000,G=1.000000,B=0.036596,A=1.000000)" 11.0)
            ;;[1 0.08817 0] Temp: delay to deal with ocean flood level being set after playlist is ready
            (Utilities|FlowControl|Delay 6.0)
            ;;[1 1 1] Save Ocean Height
            (Variables|Default|SetOceanHeight (Water|GetOceanTotalHeight (WorldSubsystems|GetWaterSubsystem)))
            ;;[1 1 1] EQS for spawning fish in the ocean
            (bind _returnvalue (AI|EQS|RunEQSQuery "/Game/Athena/Items/EnvironmentalItems/FlopperSpawn/EQS_Athena_FlopperSpawn_OceanPlacement.EQS_Athena_FlopperSpawn_OceanPlacement" (Game|GetGameState) "AllMatching"))
            (AI|EQS|SetNamedParam _returnvalue "SimpleGrid.GridSize" (ScalableFloat|GetValueatLevel (Variables|Default|GetGridSize)))
            (AI|EQS|SetNamedParam _returnvalue "SimpleGrid.SpaceBetween" (ScalableFloat|GetValueatLevel (Variables|Default|GetSpaceBetween)))
            (AI|EQS|BindEventtoOnQueryFinished _returnvalue (delegate Custom|EQS%20Finished))
            (Development|PrintString (Utilities|String|ToString(Float) (Water|GetOceanTotalHeight (WorldSubsystems|GetWaterSubsystem))) true true "(R=1.000000,G=0.137742,B=0.000000,A=1.000000)" 11.0))
          (:false
            (Actor|SetLifeSpan (Math|Random|RandomFloatinRange 0.01 0.2))))))))

(event Custom|EQS%20Finished
  (params (QueryInstance :type EQS%20Query%20Instance%20Object%20Reference) (QueryStatus :type EEnvQueryStatus%20Enum :default "Processing"))
  ;[1 1 1] Find Locations then start spawn
  (Variables|Default|SetQueryInstance QueryInstance)
  (Development|PrintString (Utilities|String|EnumtoString QueryStatus))
  (bind _resultlocations (AI|EQS|GetQueryResultsasLocations QueryInstance))
  (Variables|Default|SetFloodLocationsToSpawn _resultlocations)
  (CallFunction|StartEQSSpawn))

(event Custom|Spawn
  (params (SpawnTransform :type Transform) (ItemIndex :type Integer))
  ;[1 1 1] Spawn
  (Variables|Default|SetFlopperSpawner "/Game/Athena/Items/EnvironmentalItems/FlopperSpawn/BGA_Athena_FlopperSpawn_World.BGA_Athena_FlopperSpawn_World_C")
  (bind _asfort_game_state_athena (Utilities|Casting|CastToFortGameStateAthena (Game|GetGameState)))
  (bind _returnvalue (POI|GetPoiGridTagsforLocation _asfort_game_state_athena (.location SpawnTransform)))
  (bind _keys (Utilities|Map|Keys (Variables|Spawning|GetPOI_SpawnerMap)))
  (for _array_element _keys
    (if (GameplayTags|HasTag _returnvalue _array_element true)
      _))
  (Building|SpawnBuildingGameplayActor (Variables|Default|GetFlopperSpawner) SpawnTransform self))

(event Custom|StartEQSSpawn
  (bind _amountspawned (Variables|Spawning|GetAmountSpawned))
  (bind (_output_x _output_y _output_z) (Math|Vector|BreakVector (Utilities|Array|Get(aref) (Variables|Default|GetFloodLocationsToSpawn) _amountspawned)))
  (bind _floodlocationstospawn (Variables|Default|GetFloodLocationsToSpawn))
  ;[1 1 1] StartEQSSpawn - Loop through and spawn the fishing pools. Run a test to see if we should spawn this one. If we fail, still increment but don't spawn
  (Variables|Spawning|SetAmountSpawned 0)
  (if (<= (+ (Variables|Spawning|GetAmountSpawned) 1) (Utilities|Array|Length _floodlocationstospawn))
    (Utilities|SpawnThisTime
      (:true
        (Variables|Default|SetSpawnPoint (Math|Transform|MakeTransform (Math|Vector|MakeVector (+ _output_x (Math|Random|RandomFloatinRange -1.85e+03 1.85e+03)) (+ _output_y (Math|Random|RandomFloatinRange -1.85e+03 1.85e+03)) (+ _output_z (Variables|Spawning|GetHeightOffset))) (0.0 0.0 0.0) "1.000000,1.000000,1.000000"))
        (CallFunction|Spawn (Variables|Default|GetSpawnPoint) _amountspawned)
        (Math|Integer|IncrementInt (Variables|Spawning|GetAmountSpawned)
          (:%20%20
            (Utilities|FlowControl|Delay 0.01))))
      (:false
        _)))
  (CallFunction|StartWaterBodySpawn))

(event Custom|StartWaterBodySpawn
  (bind _waterbodyamountspawned (Variables|Default|GetWaterBodyAmountSpawned))
  (bind (_output_x _output_y _output_z) (Math|Vector|BreakVector (Utilities|Array|Get(aref) (Variables|Spawning|GetWaterBodyLocationsToSpawn) _waterbodyamountspawned)))
  (bind _waterbodylocationstospawn (Variables|Spawning|GetWaterBodyLocationsToSpawn))
  ;[1 1 1] StartSpawn - Loop through and spawn the fishing pools. Run a test to see if we should spawn this one. If we fail, still increment but don't spawn
  (Variables|Default|SetWaterBodyAmountSpawned 0)
  (if (<= (+ (Variables|Default|GetWaterBodyAmountSpawned) 1) (Utilities|Array|Length _waterbodylocationstospawn))
    (if (< _output_z (- (Variables|Default|GetOceanHeight) 1e+02))
      (Math|Integer|IncrementInt (Variables|Default|GetWaterBodyAmountSpawned)
        (:%20%20
          (Utilities|FlowControl|Delay (Math|Random|RandomFloatinRange 0.05 0.2)))))
    (CallFunction|Spawn (Math|Transform|MakeTransform (Math|Vector|MakeVector _output_x _output_y (+ _output_z (Variables|Spawning|GetHeightOffset))) (0.0 0.0 0.0) "1.000000,1.000000,1.000000") _waterbodyamountspawned)))

(fn FindCustomSplineLocations
  (for _array_element (Variables|WaterBodies|GetCustomSpawnerSpline)
    (Variables|Spawning|SetSplineLength (Spline|GetSplineLength (Variables|/Game/Athena/Items/EnvironmentalItems/FlopperSpawn/B_FlopperSpawn_CustomSpline.B_FlopperSpawn_CustomSpline_C|GetSplinePlacement _array_element)))
    (bind _output_get (Variables|Spawning|SetInitialDistanceOnSpline (Math|Random|RandomFloatinRange (Variables|WaterBodies|GetCustomSplineSpawnOffset) (+ (Variables|WaterBodies|GetCustomSplineSpawnOffset) (Variables|Spawning|GetInitialSpawnVarience)))))
    (for _index (range (+ (Math|Float|Floor (/ (- (Spline|GetSplineLength (Variables|/Game/Athena/Items/EnvironmentalItems/FlopperSpawn/B_FlopperSpawn_CustomSpline.B_FlopperSpawn_CustomSpline_C|GetSplinePlacement _array_element)) _output_get) (Variables|WaterBodies|GetDistanceBetweenSpawnsCustomSpline))) 1))
      (Utilities|SpawnThisTime
        (:true
          (Utilities|Array|AddUnique (Variables|Spawning|GetWaterBodyLocationsToSpawn) (+ (Spline|GetLocationatDistanceAlongSpline (Variables|/Game/Athena/Items/EnvironmentalItems/FlopperSpawn/B_FlopperSpawn_CustomSpline.B_FlopperSpawn_CustomSpline_C|GetSplinePlacement _array_element) (+ (* _index (Variables|WaterBodies|GetDistanceBetweenSpawnsCustomSpline)) (Variables|Spawning|GetInitialDistanceOnSpline)) "World") (Math|Vector|MakeVector (Math|Random|RandomFloatinRange (* (Variables|WaterBodies|GetCustomSplineSpawnOffset) -1.0) (Variables|WaterBodies|GetCustomSplineSpawnOffset)) (Math|Random|RandomFloatinRange (* (Variables|WaterBodies|GetCustomSplineSpawnOffset) -1.0) (Variables|WaterBodies|GetCustomSplineSpawnOffset))))))))
    (Utilities|Array|AddUnique (Variables|Spawning|GetWaterBodyLocationsToSpawn) (+ (Spline|GetLocationatDistanceAlongSpline (Variables|/Game/Athena/Items/EnvironmentalItems/FlopperSpawn/B_FlopperSpawn_CustomSpline.B_FlopperSpawn_CustomSpline_C|GetSplinePlacement _array_element) (+ (* (.ReturnValue (Math|Float|Division(WholeandRemainder) (- (Spline|GetSplineLength (Variables|/Game/Athena/Items/EnvironmentalItems/FlopperSpawn/B_FlopperSpawn_CustomSpline.B_FlopperSpawn_CustomSpline_C|GetSplinePlacement _array_element)) _output_get) (Variables|WaterBodies|GetDistanceBetweenSpawnsCustomSpline))) (Variables|WaterBodies|GetDistanceBetweenSpawnsCustomSpline)) (+ (Variables|Spawning|GetInitialDistanceOnSpline) (Math|Float|Division(WholeandRemainder) (- (Spline|GetSplineLength (Variables|/Game/Athena/Items/EnvironmentalItems/FlopperSpawn/B_FlopperSpawn_CustomSpline.B_FlopperSpawn_CustomSpline_C|GetSplinePlacement _array_element)) _output_get) (Variables|WaterBodies|GetDistanceBetweenSpawnsCustomSpline)))) "World") (Math|Vector|MakeVector (Math|Random|RandomFloatinRange (* (Variables|WaterBodies|GetCustomSplineSpawnOffset) -1.0) (Variables|WaterBodies|GetCustomSplineSpawnOffset)) (Math|Random|RandomFloatinRange (* (Variables|WaterBodies|GetCustomSplineSpawnOffset) -1.0) (Variables|WaterBodies|GetCustomSplineSpawnOffset))))))
  (return))

(fn FindIslandLocations
  (for _array_element (Variables|WaterBodies|GetIsland)
    (Utilities|IsValid _array_element
      (:Is%20Valid
        (Variables|Spawning|SetSplineLength (Spline|GetSplineLength (Variables|/Script/Water.WaterBodyIsland|GetSplineComp _array_element)))
        (bind _output_get (Variables|Spawning|SetInitialDistanceOnSpline (Math|Random|RandomFloatinRange 0.0 (Variables|Spawning|GetInitialSpawnVarience))))
        (for _index (range (+ (Math|Float|Floor (/ (- (Spline|GetSplineLength (Variables|/Script/Water.WaterBodyIsland|GetSplineComp _array_element)) _output_get) (Variables|WaterBodies|GetDistanceBetweenSpawnsIsland))) 1))
          (Utilities|SpawnThisTime
            (:true
              (Utilities|Array|AddUnique (Variables|Spawning|GetWaterBodyLocationsToSpawn) (+ (Spline|GetLocationatDistanceAlongSpline (Variables|/Script/Water.WaterBodyIsland|GetSplineComp _array_element) (+ (* _index (Variables|WaterBodies|GetDistanceBetweenSpawnsIsland)) (Variables|Spawning|GetInitialDistanceOnSpline)) "World") (* (CallFunction|GetNormal (Variables|/Script/Water.WaterBodyIsland|GetSplineComp _array_element) (+ (* _index (Variables|WaterBodies|GetDistanceBetweenSpawnsIsland)) (Variables|Spawning|GetInitialDistanceOnSpline))) (Variables|WaterBodies|GetIslandSpawnOffset))))))))) ; Fixes editor spam in test maps
  )
  (return))

(fn FindLakeLocations
  (for _array_element (Variables|WaterBodies|GetLakes)
    (Variables|Spawning|SetSplineLength (Spline|GetSplineLength (Variables|/Script/Water.WaterBody|GetSplineComp _array_element)))
    (bind _output_get (Variables|Spawning|SetInitialDistanceOnSpline (Math|Random|RandomFloatinRange 0.0 (Variables|Spawning|GetInitialSpawnVarience))))
    (for _index (range (+ (Math|Float|Floor (/ (- (Spline|GetSplineLength (Variables|/Script/Water.WaterBody|GetSplineComp _array_element)) _output_get) (Variables|WaterBodies|GetDistanceBetweenSpawnsLake))) 1))
      (Utilities|SpawnThisTime
        (:true
          (Utilities|Array|AddUnique (Variables|Spawning|GetWaterBodyLocationsToSpawn) (+ (Spline|GetLocationatDistanceAlongSpline (Variables|/Script/Water.WaterBody|GetSplineComp _array_element) (+ (* _index (Variables|WaterBodies|GetDistanceBetweenSpawnsLake)) (Variables|Spawning|GetInitialDistanceOnSpline)) "World") (* (CallFunction|GetNormal (Variables|/Script/Water.WaterBody|GetSplineComp _array_element) (+ (* _index (Variables|WaterBodies|GetDistanceBetweenSpawnsLake)) (Variables|Spawning|GetInitialDistanceOnSpline))) (* (Variables|WaterBodies|GetLakeSpawnOffset) -1.0))))))))
  (return))

(fn FindOceanLocations
  (for _array_element (Variables|WaterBodies|GetOceansIslands)
    (Variables|Spawning|SetSplineLength (Spline|GetSplineLength (Water|GetWaterSpline _array_element :!class /Script/Water.WaterBodyIsland)))
    (bind _output_get (Variables|Spawning|SetInitialDistanceOnSpline (Math|Random|RandomFloatinRange 0.0 (Variables|Spawning|GetInitialSpawnVarience))))
    (for _index (range (+ (Math|Float|Floor (/ (- (Spline|GetSplineLength (Water|GetWaterSpline _array_element :!class /Script/Water.WaterBodyIsland)) _output_get) (Variables|WaterBodies|GetDistanceBetweenSpawnsOceanIsland))) 1))
      (Utilities|SpawnThisTime
        (:true
          (Utilities|Array|AddUnique (Variables|Spawning|GetWaterBodyLocationsToSpawn) (+ (Spline|GetLocationatDistanceAlongSpline (Water|GetWaterSpline _array_element :!class /Script/Water.WaterBodyIsland) (+ (* _index (Variables|WaterBodies|GetDistanceBetweenSpawnsOceanIsland)) (Variables|Spawning|GetInitialDistanceOnSpline)) "World") (* (CallFunction|GetNormal (Water|GetWaterSpline _array_element :!class /Script/Water.WaterBodyIsland) (+ (* _index (Variables|WaterBodies|GetDistanceBetweenSpawnsOceanIsland)) (Variables|Spawning|GetInitialDistanceOnSpline))) (Variables|WaterBodies|GetOceanIslandSpawnOffset))))))))
  (return))

(fn FindRiverNarrowLocations
  (for _array_element (Variables|WaterBodies|GetRiversNarrow)
    (Variables|Spawning|SetSplineLength (Spline|GetSplineLength (Variables|/Script/Water.WaterBody|GetSplineComp _array_element)))
    (bind _output_get (Variables|Spawning|SetInitialDistanceOnSpline (Math|Random|RandomFloatinRange (Variables|WaterBodies|GetRiverNarrowSpawnOffset) (+ (Variables|WaterBodies|GetRiverNarrowSpawnOffset) (Variables|Spawning|GetInitialSpawnVarience)))))
    (for _index (range (+ (Math|Float|Floor (/ (- (Spline|GetSplineLength (Variables|/Script/Water.WaterBody|GetSplineComp _array_element)) _output_get) (Variables|WaterBodies|GetDistanceBetweenSpawnsRiverNarrow))) 1))
      (Utilities|SpawnThisTime
        (:true
          (Utilities|Array|AddUnique (Variables|Spawning|GetWaterBodyLocationsToSpawn) (+ (Spline|GetLocationatDistanceAlongSpline (Variables|/Script/Water.WaterBody|GetSplineComp _array_element) (+ (* _index (Variables|WaterBodies|GetDistanceBetweenSpawnsRiverNarrow)) (Variables|Spawning|GetInitialDistanceOnSpline)) "World") (Math|Vector|MakeVector (Math|Random|RandomFloatinRange (* (Variables|WaterBodies|GetRiverNarrowSpawnOffset) -1.0) (Variables|WaterBodies|GetRiverNarrowSpawnOffset)) (Math|Random|RandomFloatinRange (* (Variables|WaterBodies|GetRiverNarrowSpawnOffset) -1.0) (Variables|WaterBodies|GetRiverNarrowSpawnOffset)))))))))
  (return))

(fn FindRiverWideLocations
  (for _array_element (Variables|WaterBodies|GetRiversWide)
    (Variables|Spawning|SetSplineLength (Spline|GetSplineLength (Variables|/Script/Water.WaterBody|GetSplineComp _array_element)))
    (bind _output_get (Variables|Spawning|SetInitialDistanceOnSpline (Math|Random|RandomFloatinRange (Variables|WaterBodies|GetRiverWideSpawnOffset) (+ (Variables|WaterBodies|GetRiverWideSpawnOffset) (Variables|Spawning|GetInitialSpawnVarience)))))
    (for _index (range (+ (Math|Float|Floor (/ (- (Spline|GetSplineLength (Variables|/Script/Water.WaterBody|GetSplineComp _array_element)) _output_get) (Variables|WaterBodies|GetDistanceBetweenSpawnsRiverWide))) 1))
      (Utilities|SpawnThisTime
        (:true
          (Utilities|Array|AddUnique (Variables|Spawning|GetWaterBodyLocationsToSpawn) (+ (Spline|GetLocationatDistanceAlongSpline (Variables|/Script/Water.WaterBody|GetSplineComp _array_element) (+ (* _index (Variables|WaterBodies|GetDistanceBetweenSpawnsRiverWide)) (Variables|Spawning|GetInitialDistanceOnSpline)) "World") (Math|Vector|MakeVector (Math|Random|RandomFloatinRange (* (Variables|WaterBodies|GetRiverWideSpawnOffset) -1.0) (Variables|WaterBodies|GetRiverWideSpawnOffset)) (Math|Random|RandomFloatinRange (* (Variables|WaterBodies|GetRiverWideSpawnOffset) -1.0) (Variables|WaterBodies|GetRiverWideSpawnOffset)))))))))
  (return))

(fn GetNormal
  (params (Spline :type Spline%20Component%20Object%20Reference) (Distance :type Float%20%28double-precision%29))
  (return (Math|Vector|Normalize (Math|Vector|CrossProduct (Spline|GetTangentatDistanceAlongSpline Spline Distance "World") (Spline|GetUpVectoratDistanceAlongSpline Spline Distance "World")))))

(fn RollRarity
  (return "Legendary"))

(fn ConstructionScript)
