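; EQS context for Athena flopper spawning that provides a single origin location for spawn queries.
; When the querier actor is valid, it picks random X and Y within a square around the world origin, without using the querier's position.
; Z comes from the Water subsystem's total ocean height, lowered slightly to sit just below the surface.
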(fn ProvideSingleLocation
  (params (QuerierObject :type Object%20Reference) (QuerierActor :type Actor%20Object%20Reference))
  (Utilities|IsValid QuerierActor
    (:Is%20Valid
      (return (Math|Random|RandomFloatinRange -1e+03 1e+03) (Math|Random|RandomFloatinRange -1e+03 1e+03) (- (Water|GetOceanTotalHeight (WorldSubsystems|GetWaterSubsystem)) 25.0)))))
