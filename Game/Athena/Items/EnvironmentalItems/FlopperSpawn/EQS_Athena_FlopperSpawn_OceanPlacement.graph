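# Environment query that picks ocean placement points for Flopper spawns in Athena.
# A large grid with query-parameter size and spacing surrounds a Flopper spawn origin context, with trace projection disabled.
# Overlap boxes require a hit in a wide, shallow volume below each point, also scored, and none in a narrow column above.
# Points must also lie inside a water body and overlap a second trace channel using complex collision.
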EnvQuery /Game/Athena/Items/EnvironmentalItems/FlopperSpawn/EQS_Athena_FlopperSpawn_OceanPlacement
options [{
  generator {
   name=EnvQueryGenerator_SimpleGrid_1
   class=EnvQueryGenerator_SimpleGrid
   properties {
    generateAround=&/EQSC_Athena_FlopperSpawn_Origin.EQSC_Athena_FlopperSpawn_Origin_C
    gridSize {
     dataBinding=&self:EnvQueryGenerator_SimpleGrid_1.AIDataProvider_QueryParams_2
     dataField=FloatValue
     defaultValue=160000
    }
    projectionData {
     extentX=0
     navigationFilter=null
     postProjectionVerticalOffset=0
     projectDown=1024
     projectUp=1024
     serializedChannel=ECC_GameTraceChannel11
     traceChannel=TraceTypeQuery10
     traceMode=None
    }
    spaceBetween {
     dataBinding=&self:EnvQueryGenerator_SimpleGrid_1.AIDataProvider_QueryParams_3
     dataField=FloatValue
     defaultValue=5500
    }
   }
  }
  tests [
   EnvQueryTest_Overlap_0 {
    class=EnvQueryTest_Overlap
    properties {
     boolValue=true
     scoringFactor=2
     testPurpose=FilterAndScore
     overlapData {extentX=3000, extentY=3000, extentZ=150, overlapChannel=ECC_GameTraceChannel11, overlapShape=Box, shapeOffset=<0 0 -75>}
    }
   }
   EnvQueryTest_Overlap_1 {
    class=EnvQueryTest_Overlap
    properties {overlapData {extentX=10, extentY=10, extentZ=500, overlapChannel=ECC_GameTraceChannel11, overlapShape=Box, shapeOffset=<0 0 250>}}
   }
   EnvQueryTest_InsideWaterBody_0 {
    class=EnvQueryTest_InsideWaterBody
    properties {
     bDefineReferenceValue=false
     bIgnoreExclusionVolumes=false
     bIncludeWaves=false
     bSimpleWaves=false
     boolValue=true
     clampMaxType=None
     clampMinType=None
     filterType=Match
     floatValueMax=0
     floatValueMin=0
     multipleContextFilterOp=AllPass
     multipleContextScoreOp=AverageScore
     normalizationType=Absolute
     referenceValue=0
     scoreClampMax=0
     scoreClampMin=0
     scoringEquation=Constant
     scoringFactor=1
     testComment=""
     testPurpose=Filter
    }
   }
   EnvQueryTest_Overlap_2 {
    class=EnvQueryTest_Overlap
    properties {
     boolValue=true
     overlapData {
      bOnlyBlockingHits=false
      bOverlapComplex=true
      extentX=25
      extentY=25
      extentZ=25
      overlapChannel=ECC_GameTraceChannel15
      overlapShape=Box
      shapeOffset=<0 0 0>
     }
    }
   }
  ]
 }]
